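# Translucent glass shard material for the Rift Launcher portal pickup, usable on skeletal meshes and Niagara mesh particles.
# It fakes reflections with a view-space sphere map plus a sun/moon specular highlight, tinted by time of day and ambient cubemap.
# That reflection blends toward a Fresnel rim color and passes through an emissive-adjustment function.
# A Fresnel-lifted glass color supplies the base tint, and the tinted reflection also drives a scalable opacity term.
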Material /HopscotchWeaponsGameplay/Items/GrenadeLauncher/FX/RiftLauncher/M_Shard_Portal_Pickup
graph {
 properties {
  bUsedWithMeshDeformer=true
  bUsedWithNiagaraMeshParticles=true
  bUsedWithSkeletalMesh=true
  blendMode=BLEND_Translucent
  pixelDepthOffsetMode=PDOM_Legacy
  translucencyLightingMode=TLM_VolumetricPerVertexNonDirectional
  usedShadingModels=Default Lit
 }
 nodes [
  Add(
   A=Multiply(A=Fresnel() {baseReflectFraction=0, exponent=4}) {constB=0.5}
   B=VectorParameter() {parameterName=GlassColor, defaultValue=<rgba 0 0.0275 0.5647 0>}.RGB
  )
  n1 = ScalarParameter() {defaultValue=7.5, parameterName=Fresnel Exponent}
  n2 = ComponentMask(None=Transform(None=ReflectionVectorWS()) {transformSourceType=TRANSFORMSOURCE_World, transformType=TRANSFORM_View}) {g=true, r=true}
  n3 = Constant2Vector() {g=-0.4, r=0.4}
  n4 = TextureSampleParameter2D(UVs=Add(A=Multiply(A=@n2, B=@n3), B=Constant2Vector() {g=0.5, r=0.5})) {parameterName=SphereReflectionMap, texture=/Game/Items/Art_noLOD/Textures/T_Items_MineralPowder_Glass_Reflection}
  n5 = CollectionParameter() {
   collection=&FortniteMaterialParameters
   parameterName=SunAndMoonModelDirectionalVector
  }
  n6 = ComponentMask(None=@n5) {b=true, g=true, r=true}
  n7 = ReflectionVectorWS()
  n8 = Power(Base=Clamp(None=DotProduct(A=@n6, B=@n7)), Exp=ScalarParameter() {defaultValue=70, parameterName=Spec Highlight Power}) {constExponent=20}
  n9 = ScalarParameter() {defaultValue=3, parameterName=Spec Intense}
  n10 = Multiply(
   A=Add(A=MaterialFunctionCall() {materialFunction=&Color_Time_of_Day}.Result) {constB=0.5}
   B=Add(A=Multiply(A=@n4.RGB) {constB=0.25}, B=Multiply(A=@n8, B=@n9))
  )
  n11 = MaterialFunctionCall() {materialFunction=/Engine/Functions/GetPostProcessFunctions/GetAmbientCubemapTint}
  n12 = Multiply(A=Multiply(A=@n10, B=ComponentMask(None=@n11.AmbientCubemapTint) {b=true, g=true, r=true})) {constB=0.4}
  MaterialFunctionCall(Emissive=LinearInterpolate(
    A=Add(A=@n10, B=@n12)
    B=VectorParameter() {parameterName=FresnelColor, defaultValue=<rgba 2.7 3.4437 18 0>}.RGB
    Alpha=Fresnel(ExponentIn=@n1) {baseReflectFraction=0, exponent=4}
   )) {materialFunction=&MF_FX_EmissiveAdjustments}
  Reroute(None=Multiply(A=Add(A=@n12) {constB=0.5}, B=ScalarParameter() {defaultValue=1, parameterName=Opacity}))
 ]
}
